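# Base Editor Utility Widget for tech-art tools that picks files, directories, assets or content paths; it has no layout, and getter functions supply the button, text-field and label widgets.
# The main button opens a Python-driven OS dialog, while an optional second button takes the Content Browser selection.
# Picked, typed and starting paths flow through setCurrentPaths, which applies the absolute, relative or asset scope via path converters, updates text and tooltip, and broadcasts onPathsChanged.
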EditorUtilityWidgetBlueprint /Game/Blueprints/Blutilities/ForTechArt/Widgets/FilePicker/FilePicker_Base_EUW
graphs [
 EventGraph {nodes [
   initButtonEvents() {is=title K2Node_CustomEvent}
   > n1 = Sequence()
   > Is Valid(InputObject=getButtonWidget!() -> (button:Widget Object Reference))
   n2 = getButtonWidget()
   > n3 = Cast To TA_BaseEditorButtonWidget!(execute, Object=@n2) -> (then, CastFailed, AsTA Base Editor Button Widget:TA Base Editor Button Widget Object Reference) {is=title K2Node_DynamicCast}
   n4 = handleButtonClick_asTA!() -> (OutputDelegate, then, widget:User Widget Object Reference, event:Pointer Event Structure) {is=title K2Node_CustomEvent}
   Bind Event to On Clicked(execute=@n3.then, self:TA Base Editor Button Widget Object Reference=@n3, Delegate=@n4) {is=title K2Node_AddDelegate}
   n5 = Cast To Button!(execute=@n3.CastFailed, Object=@n2) -> (then, CastFailed, AsButton:Button Object Reference) {is=title K2Node_DynamicCast}
   n6 = handleButtonClick_asSlateButton() {is=title K2Node_CustomEvent}
   Assign On Clicked!(execute:Exec=@n5.then, self:Button Object Reference=@n5, Delegate:"Delegate (by ref)"=@n6) -> (then:Exec) {is=title K2Node_AssignDelegate}
   n7 = getAssetPickerButton!() -> (widget:Widget Object Reference)
   Is Valid(exec=@n1.then_1, InputObject=@n7)
   > n8 = Cast To TA_BaseEditorButtonWidget(Object=@n7)
   n9 = CustomEvent_0!() -> (OutputDelegate, then, widget:User Widget Object Reference, event:Pointer Event Structure) {is=title K2Node_CustomEvent}
   Bind Event to On Clicked(execute=@n8.then, self=@n8, Delegate=@n9)
   n10 = Cast To Button(execute=@n8.CastFailed, Object=@n7)
   n11 = HandlePicker_asButton() {is=title K2Node_CustomEvent}
   Assign On Clicked(execute=@n10.then, self=@n10, Delegate=@n11)
   Event PreConstruct()
   > initLabel()
   > setCurrentPaths!(execute, self, CurrentPaths:Array of Strings=Get startingPaths() -> (startingPaths:Array of Strings)) -> (then)
   Event Construct()
   > validateEnabledVisuals()
   > initButtonEvents() {is=title K2Node_CallFunction}
   > n12 = Cast To EditableText!(execute, Object=getInputWidget!() -> (widget:Widget Object Reference)) -> (then, CastFailed, AsEditable Text:Editable Text Object Reference) {is=title K2Node_DynamicCast}
   > "Bind Event to OnTextCommitted (Editable Text)"(self:Editable Text Object Reference=@n12, Delegate=Create Event()) {is=title K2Node_AddDelegate}
   openPicker(execute=@n6.then)
   openPicker(execute=@n4.then)
   chooseAssetFromContentBrowser(execute=@n11.then)
   chooseAssetFromContentBrowser(execute=@n9.then)
  ]}
 GetFirstCurrentPath {nodes [
   GetFirstCurrentPath() {is=title K2Node_FunctionEntry}
   n1 = Get CurrentPaths() -> (CurrentPaths:Array of Strings)
   > n2 = Branch(Condition="integer > integer"(A=Length(TargetArray:Array of Strings=@n1)))
   > Return Node!(execute, CurrentPath:String="Get (a ref)"(Array:Array of Strings=Get CurrentPaths(), Dimension 1=0) -> (Output:"String (by ref)"))
   Return Node(execute=@n2.else)
  ]}
 RefreshFilePickerDisplay {nodes [
   RefreshFilePickerDisplay() {is=title K2Node_FunctionEntry}
   > setCurrentPaths(CurrentPaths=Get CurrentPaths())
   > validateEnabledVisuals()
  ]}
 chooseAssetFromContentBrowser {nodes [
   chooseAssetFromContentBrowser() {is=title K2Node_FunctionEntry}
   > n1 = Sequence!(execute) -> (then_0, then_1, then_2:Exec)
   > n2 = Switch on TA_FilePicker_Type_Enum!(execute, Selection:TA_FilePicker_Type_Enum Enum=Get pickerType() -> (pickerType:TA_FilePicker_Type_Enum Enum=NewEnumerator0)) -> (NewEnumerator0:Exec, NewEnumerator1:Exec) {is=title K2Node_SwitchEnum}
   > n3 = GetSelectedAssetData()
   > n4 = Branch(Condition=Get allowMultipleFilesSelected() -> (allowMultipleFilesSelected:bool))
   > n5 = For Each Loop!(Exec, Array:Array of Asset Data Structures=@n3) -> (
    LoopBody
    Array Element_PackageName:Name
    Array Element_PackagePath:Name
    Array Element_AssetName:Name
    Array Element_AssetClassPath:Top Level Asset Path Structure
    Array Index
    Completed
   )
   > Add(TargetArray=Get l_initialPaths() -> (l_initialPaths:Array of Strings), NewItem="To String (Name)"(InName=@n5)) {is=title K2Node_CallArrayFunction}
   Branch(execute=@n4.else, Condition=Is Valid Index!(TargetArray:Array of Asset Data Structures=@n3, IndexToTest:int) -> (ReturnValue:bool))
   > Add(
    TargetArray=Get l_initialPaths()
    NewItem="To String (Name)"(InName="Get (a copy)"!(Array:Array of Asset Data Structures=@n3, Dimension 1) -> (Output_PackageName:Name, Output_PackagePath:Name, Output_AssetName:Name, Output_AssetClassPath:Top Level Asset Path Structure))
   )
   n6 = GetSelectedAssetData(execute=@n2.NewEnumerator1)
   > Branch(Condition=Is Valid Index(TargetArray=@n6))
   > Add(TargetArray=Get l_initialPaths(), NewItem="To String (Name)"(InName="Get (a copy)"(Array=@n6).Output_PackagePath))
   n7 = For Each Loop!(Exec=@n1.then_1, Array:Array of Strings=Get l_initialPaths()) -> (LoopBody, Array Element:String, Array Index, Completed)
   > n8 = Switch on TA_FilePicker_Scope_Enum!(execute, Selection:TA_FilePicker_Scope_Enum Enum=Get pathScope() -> (pathScope:TA_FilePicker_Scope_Enum Enum=NewEnumerator0)) -> (NewEnumerator0:Exec, NewEnumerator1:Exec, NewEnumerator2:Exec) {is=title K2Node_SwitchEnum}
   n9 = convertAssetToFilePath!(execute=[@n8.NewEnumerator0, @n8.NewEnumerator1], self, path:String=@n7) -> (then, asset:String)
   > Add(TargetArray=Get l_returnPaths() -> (l_returnPaths:Array of Strings), NewItem=@n9)
   Set l_returnPaths(execute=@n8.NewEnumerator2, l_returnPaths:Array of Strings=Get l_initialPaths()) -> (Output_Get:Array of Strings)
   setCurrentPaths(execute=@n1.then_2, CurrentPaths=Get l_returnPaths())
  ]}
 convertAssetToFilePath {nodes [
   n1 = convertAssetToFilePath!() -> (then, path:String) {is=title K2Node_FunctionEntry}
   > Set l_path(l_path:String=Replace(SourceString=@n1, From="\\", To=/)) -> (Output_Get:String)
   > n2 = GetAssetsByPackageName!(
    execute
    self:Asset Registry Interface=GetAssetRegistry()
    PackageName:Name=String To Name(InString=Get l_path() -> (l_path:String))
    bIncludeOnlyOnDiskAssets:bool=false
    bSkipARFilteredAssets:bool=true
   ) -> (then, OutAssetData:Array of Asset Data Structures, ReturnValue:bool)
   > n3 = Branch(Condition=Is Valid Index(TargetArray=@n2))
   > K2Node_Composite_0 = HandleAssetExists!(execute, packageName:String=Get l_path()) -> (then:Exec, path:String) {is=title K2Node_Composite}
   > n4 = Set l_path(l_path=@K2Node_Composite_0)
   n5 = Get l_path()
   PrintString(execute=[@n6.then, @n4.then], InString=Append(A="Final to open: ", B=@n5))
   > Return Node!(execute, asset:String=@n5)
   K2Node_Composite_2 = HandleNoAssetExists!(execute=@n3.else, directory:String=Get l_path()) -> (then:Exec, path:String) {is=title K2Node_Composite}
   > n6 = Set l_path(l_path=@K2Node_Composite_2)
  ]}
 HandleAssetExists {
  path=convertAssetToFilePath.K2Node_Composite_0.HandleAssetExists
  nodes [
   n1 = Inputs!() -> (execute, packageName:String)
   > Set l_path(l_path=@n1)
   > n2 = Sequence!(execute) -> (then_0, then_1, then_2, then_3:Exec)
   > n3 = doAssetToFileStringOperations!(execute, self, inPath:String=Get l_path()) -> (then, out:String)
   > Set l_path(l_path=@n3)
   Set l_path(execute=@n2.then_1, l_path=Append(A=Get l_path(), B=.uasset))
   n4 = Branch(execute=@n2.then_2, Condition=FileExists!(InPath:String=Get l_path()) -> (ReturnValue:bool))
   > n5 = Switch on TA_FilePicker_Type_Enum(Selection=Get pickerType() -> (pickerType=NewEnumerator0))
   Set l_path(execute=@n5.NewEnumerator1, l_path=GetPath!(InPath:String=Get l_path()) -> (ReturnValue:String))
   Set l_path(execute=@n4.else)
   Outputs!(then:Exec=@n2.then_3, path:String=Get l_path())
  ]
 }
 HandleNoAssetExists {
  path=convertAssetToFilePath.K2Node_Composite_2.HandleNoAssetExists
  nodes [
   n1 = Inputs!() -> (execute, directory:String)
   > n2 = Set l_path(l_path=@n1)
   > n3 = Sequence!(execute) -> (then_0, then_1)
   > n4 = Branch(Condition=DirectoryExists!(InPath:String=@n2) -> (ReturnValue:bool))
   n5 = doAssetToFileStringOperations(execute=@n4.else, inPath=Get l_path())
   > Set l_path(l_path=@n5)
   Outputs(then=@n3.then_1, path=Get l_path())
  ]
 }
 convertFilePathToAsset {nodes [
   n1 = convertFilePathToAsset!() -> (then, path:String) {is=title K2Node_FunctionEntry}
   > Set l_path(l_path=@n1)
   > n2 = Sequence()
   > n3 = Switch on TA_FilePicker_Type_Enum(Selection=Get pickerType() -> (pickerType=NewEnumerator0))
   n4 = GetAssetRegistry()
   > n5 = GetAssetsByPackageName(self=@n4, PackageName=String To Name(InString=Get l_path()), bIncludeOnlyOnDiskAssets=false, bSkipARFilteredAssets=true)
   > n6 = Branch(Condition=Is Valid Index(TargetArray=@n5))
   K2Node_Composite_0 = DoStringFileConvert!(execute=[@n11.then_0, @n6.else], inString:String=Get l_path()) -> ("  ":Exec, outString:String) {is=title K2Node_Composite}
   > Set l_path(l_path=@K2Node_Composite_0)
   n7 = GetAllCachedPaths!(execute=@n3.NewEnumerator1, self:Asset Registry Interface=[GetAssetRegistry(), @n4]) -> (then, OutPathList:Array of Strings)
   > n8 = Branch(Condition=Contains Item(TargetArray:Array of Strings=@n7, ItemToFind:"String (by ref)"=Get l_path()))
   n9 = GetAssetsByPackageName(
    execute=@n8.else
    self=GetAssetRegistry()
    PackageName=String To Name(InString=Get l_path())
    bIncludeOnlyOnDiskAssets=false
    bSkipARFilteredAssets=true
   )
   > n10 = Branch(Condition=Is Valid Index(TargetArray=@n9))
   > Set l_path(l_path=Split(SourceString=Get l_path(), InStr=/, SearchDir=FromEnd))
   n11 = Sequence(execute=@n10.else)
   n12 = GetAssetsByPackageName(
    execute=@n11.then_1
    self=GetAssetRegistry()
    PackageName=String To Name(InString=Get l_path())
    bIncludeOnlyOnDiskAssets=false
    bSkipARFilteredAssets=true
   )
   > Branch(Condition=Is Valid Index(TargetArray=@n12))
   > Set l_path(l_path=Split(SourceString=Get l_path(), InStr=/, SearchDir=FromStart))
   Return Node(execute=@n2.then_1, asset=Get l_path())
   Replace(SourceString=@n1, From="\\", To=/)
  ]}
 DoStringFileConvert {
  path=convertFilePathToAsset.K2Node_Composite_0.DoStringFileConvert
  nodes [
   n1 = Inputs!() -> (execute, inString:String)
   > Set l_path(l_path=@n1)
   > n2 = Sequence!(execute) -> (then_0, then_1, then_2:Exec, then_3:Exec)
   n3 = Get l_path()
   > Branch(Condition=IsRelative!(InPath:String=@n3) -> (ReturnValue:bool))
   > Set l_path(l_path=ConvertToAbsolutePath!(Filename:String=@n3) -> (ReturnValue:String))
   n4 = Split(SourceString=@n3, InStr=FortniteGame/Content, SearchDir=FromStart)
   n5 = Branch(execute=@n2.then_1, Condition=@n4.ReturnValue)
   > Set l_path(l_path=Append(A=/Game, B=@n4.RightS))
   n6 = Split(SourceString=Get l_path(), InStr=/Plugins/GameFeatures, SearchDir=FromStart)
   n7 = Branch(execute=@n5.else, Condition=@n6.ReturnValue)
   > n8 = Set l_path(l_path=@n6.RightS)
   n9 = Split(SourceString=@n8, InStr=/Content, SearchDir=FromStart)
   > Set l_path(l_path=Append(A=@n9, B=@n9.RightS))
   n10 = Split(SourceString=Get l_path(), InStr=/Engine/Content, SearchDir=FromStart)
   n11 = Branch(execute=@n7.else, Condition=@n10.ReturnValue)
   > Set l_path(l_path=Append(A=/Engine, B=@n10.RightS))
   Outputs!("  ":Exec=[@n11.else, @n2.then_3], outString:String=Get l_path())
   n12 = Split(SourceString=Get l_path(), InStr=., SearchDir=FromEnd)
   Branch(execute=@n2.then_2, Condition=@n12.ReturnValue)
   > Set l_path(l_path=@n12)
  ]
 }
 doAssetToFileStringOperations {nodes [
   n1 = doAssetToFileStringOperations!() -> (then, inPath:String) {is=title K2Node_FunctionEntry}
   > Set l_path(l_path=@n1)
   > n2 = Sequence!(execute) -> (then_0, then_1)
   n3 = Split(SourceString=Get l_path(), InStr=/Game/, SearchDir=FromStart)
   > n4 = Branch(Condition=@n3.ReturnValue)
   > Set l_path(l_path=Append(A=GetProjectContentDirectory!() -> (ReturnValue:String), B=@n3.RightS))
   n5 = Split(SourceString=Get l_path(), InStr=/Engine/, SearchDir=FromStart)
   n6 = Branch(execute=@n4.else, Condition=@n5.ReturnValue)
   > Set l_path(l_path=Append(A=EngineContentDir!() -> (ReturnValue:String), B=@n5.RightS))
   Set l_insertIndex(execute=@n6.else, l_insertIndex:int=FindSubstring(SearchIn=Get l_path(), Substring=/, StartPosition=2)) -> (Output_Get:int=0)
   n7 = Get l_path()
   n8 = Get l_insertIndex() -> (l_insertIndex:int=0)
   n9 = Left!(SourceString:String=@n7, Count:int=@n8) -> (ReturnValue:String)
   n10 = RightChop!(SourceString:String=@n7, Count:int=@n8) -> (ReturnValue:String)
   > Set l_path(l_path=Append!(A=GetProjectDirectory!() -> (ReturnValue:String), B=Plugins/GameFeatures, C:String=Append(A=@n9, B=/Content, C=@n10)) -> (ReturnValue))
   Return Node!(execute=@n2.then_1, out:String=Get l_path())
  ]}
 getAssetPickerButton {nodes [
   getAssetPickerButton() {is=title K2Node_FunctionEntry}
   > Return Node!(execute, widget:Widget Object Reference)
  ]}
 getButtonWidget {nodes [
   getButtonWidget() {is=title K2Node_FunctionEntry}
   > Return Node!(execute, button:Widget Object Reference)
  ]}
 getCurrentPathsAsNames {nodes [
   getCurrentPathsAsNames() {is=title K2Node_FunctionEntry}
   > n1 = For Each Loop(Array=Get CurrentPaths())
   Return Node!(execute=@n1.Completed, CurrentPaths:Array of Names=Get currentPathsAsNames() -> (currentPathsAsNames:Array of Names))
   > Add(TargetArray:Array of Names=Get currentPathsAsNames(), NewItem:"Name (by ref)"=String To Name(InString=@n1))
  ]}
 getInitialDirectoryPath {nodes [
   getInitialDirectoryPath() {is=title K2Node_FunctionEntry}
   > n1 = Switch on TA_FilePicker_InitDirectoryType_Enum!(
    execute
    Selection:TA_FilePicker_InitDirectoryType_Enum Enum=Get initialDirectoryType() -> (initialDirectoryType:TA_FilePicker_InitDirectoryType_Enum Enum=NewEnumerator0)
   ) -> (
    NewEnumerator0:Exec
    NewEnumerator1:Exec
    NewEnumerator2:Exec
    NewEnumerator3:Exec
    NewEnumerator6:Exec
    NewEnumerator4:Exec
    NewEnumerator5:Exec
   ) {is=title K2Node_SwitchEnum}
   > Return Node!(execute, initialDirectoryOverride:String)
   Return Node(execute=@n1.NewEnumerator1, initialDirectoryOverride=GetProjectDirectory())
   Return Node(execute=@n1.NewEnumerator2, initialDirectoryOverride=GetProjectContentDirectory())
   Return Node(execute=@n1.NewEnumerator3, initialDirectoryOverride=GetProjectSavedDirectory!() -> (ReturnValue:String))
   n2 = Branch(execute=@n1.NewEnumerator4, Condition=Is Valid Index(TargetArray:Array of Strings=Get CurrentPaths()))
   > n3 = Set l_path(l_path="Get (a copy)"!(Array:Array of Strings=Get CurrentPaths(), Dimension 1) -> (Output:String))
   > n4 = Branch(Condition=FileExists(InPath=Get l_path()))
   > Return Node(initialDirectoryOverride=GetPath(InPath=Get l_path()))
   n5 = Branch(execute=@n4.else, Condition=DirectoryExists(InPath=@n3))
   > Return Node(initialDirectoryOverride=Get l_path())
   n6 = Switch on TA_FilePicker_Scope_Enum(execute=@n5.else, Selection=Get pathScope() -> (pathScope=NewEnumerator0))
   Return Node(execute=[@n6.NewEnumerator0, @n6.NewEnumerator1])
   n7 = convertAssetToFilePath(execute=@n6.NewEnumerator2, path=Get l_path()) {is=title K2Node_CallFunction}
   > n8 = Branch(Condition=FileExists(InPath=@n7))
   > Return Node(initialDirectoryOverride=GetPath(InPath=@n7))
   Return Node(execute=@n8.else, initialDirectoryOverride=@n7)
   Return Node(execute=@n2.else)
   Return Node(execute=@n1.NewEnumerator5, initialDirectoryOverride=Get initialDirectoryOverride() -> (initialDirectoryOverride:String))
   Return Node(execute=@n1.NewEnumerator6, initialDirectoryOverride=GetGameFeaturesDirAbsolutePath!() -> (ReturnValue:String))
  ]}
 getInputWidget {nodes [
   getInputWidget() {is=title K2Node_FunctionEntry}
   > Return Node!(execute, widget:Widget Object Reference)
  ]}
 getLabelWidget {nodes [
   getLabelWidget() {is=title K2Node_FunctionEntry}
   > Return Node()
  ]}
 getPathStringToDisplay {nodes [
   getPathStringToDisplay() {is=title K2Node_FunctionEntry}
   n1 = Get CurrentPaths()
   > n2 = Branch(Condition="integer > integer"(A=Length(TargetArray=@n1)))
   n3 = Get CurrentPaths()
   > n4 = Branch(Condition="integer > integer"(A=Length(TargetArray=@n3), B=1))
   > Return Node!(execute, string:String=Multiple Files...)
   Return Node(execute=@n4.else, string="Get (a copy)"(Array=Get CurrentPaths()))
   Return Node(execute=@n2.else)
  ]}
 handleInputWidgetTextChanged {nodes [
   n1 = handleInputWidgetTextChanged!() -> (then, Text:"Text (by ref)", CommitMethod:ETextCommit Enum=Default) {is=title K2Node_FunctionEntry}
   > setCurrentPaths(CurrentPaths=Make Array("[0]":String="To String (Text)"(InText=@n1)) -> (Array:Array of Strings))
   Return Node!(execute)
  ]}
 initLabel {nodes [
   initLabel() {is=title K2Node_FunctionEntry}
   > n1 = Branch(Condition="Not Equal Exactly (String)"(A=Get WidgetLabelOverride() -> (WidgetLabelOverride:String)))
   > setLabel!(execute, self, label:String=Get WidgetLabelOverride()) -> (then)
   n2 = Switch on TA_FilePicker_Scope_Enum(execute=@n1.else, Selection=Get pathScope() -> (pathScope=NewEnumerator0))
   n3 = Get pickerType() -> (pickerType=NewEnumerator0)
   n4 = Switch on TA_FilePicker_Type_Enum(execute=[@n2.NewEnumerator0, @n2.NewEnumerator1], Selection=@n3)
   > setLabel(label=File)
   setLabel(execute=@n4.NewEnumerator1, label=Directory)
   n5 = Switch on TA_FilePicker_Type_Enum(execute=@n2.NewEnumerator2, Selection=@n3)
   > setLabel(label=Asset)
   setLabel(execute=@n5.NewEnumerator1, label=Path)
  ]}
 onPathsChanged {nodes [onPathsChanged!() -> (then, widget:File Picker Base EUW Object Reference, paths:Array of Strings, firstPath:String) {is=title K2Node_FunctionEntry}]}
 openPicker {nodes [
   openPicker() {is=title K2Node_FunctionEntry}
   > n1 = Switch on TA_FilePicker_Type_Enum(Selection=Get pickerType() -> (pickerType=NewEnumerator0))
   > n2 = Execute Python Script!(
    execute
    PythonScript="import unreal\r\nimport TechArt.Tools.FilePickerUtils as utils\r\nimport importlib\r\n\r\nimportlib.reload(utils)\r\n\r\ndir = unreal.Paths.project_content_dir()\r\n\r\nif multiple:\r\n    file_paths = utils.OpenDialog_File(initial_dir=initial_directory,title=title, file_types_map=file_types_map, multiple=True)\r\n    first_path = file_paths[0]\r\nelse:\r\n    first_path = utils.OpenDialog_File(initial_dir=initial_directory,title=title, file_types_map=file_types_map, multiple=False)\r\n    file_paths = [first_path]"
    file_types_map:Map of Strings to Strings=Get fileTypes() -> (fileTypes:Map of Strings to Strings)
    initial_directory:String=getInitialDirectoryPath!() -> (initialDirectoryOverride:String) {is=title K2Node_CallFunction}
    title:String=Get PopupTitle() -> (PopupTitle:String)
    multiple:bool=Get allowMultipleFilesSelected()
   ) -> (then, ReturnValue, file_paths:Array of Strings, first_path:String) {inputs [file_types_map, initial_directory, title, multiple], outputs [file_paths,first_path]}
   > Set l_paths(l_paths:Array of Strings=@n2.file_paths) -> (Output_Get:Array of Strings)
   > n3 = Set l_path(l_path=@n2.first_path)
   Branch(execute=[@n5.then, @n3.then], Condition="Not Equal Exactly (String)"(A=Get l_path()))
   > setCurrentPaths(CurrentPaths=Get l_paths() -> (l_paths:Array of Strings))
   n4 = Execute Python Script!(
    execute=@n1.NewEnumerator1
    PythonScript="import unreal\r\nimport TechArt.Tools.FilePickerUtils as utils\r\nimport importlib\r\n\r\nimportlib.reload(utils)\r\n\r\ndir = unreal.Paths.project_content_dir()\r\n##file_types=((\"Asset\",\"*.uasset\"),(\"Map/Level\", \"*.umap\"),(\"*\",\"*.*\"))\r\npath = utils.OpenDialog_Directory(initial_dir=initial_directory,title=title)"
    file_types_map=Get fileTypes()
    initial_directory=getInitialDirectoryPath()
    title=Get PopupTitle()
   ) -> (then, ReturnValue, path:String) {inputs [file_types_map, initial_directory, title], outputs [path]}
   > Set l_path(l_path=@n4.path)
   > n5 = Set l_paths(l_paths=Make Array("[0]"=@n4.path))
  ]}
 setCurrentPaths {nodes [
   n1 = setCurrentPaths!() -> (then, CurrentPaths:Array of Strings) {is=title K2Node_FunctionEntry}
   > Set CurrentPaths(CurrentPaths:Array of Strings=@n1) -> (Output_Get:Array of Strings)
   > n2 = Sequence()
   > n3 = For Each Loop(Array=Get CurrentPaths())
   n4 = Reverse for Each Loop!(Exec:Exec=@n3.Completed, Array:Array of Strings=Get CurrentPaths()) -> (LoopBody:Exec, ArrayIndex:int, ArrayElement:String, Completed:Exec) {is=title K2Node_MacroInstance}
   > Branch(Condition="Equal Exactly (String)"(A=@n4.ArrayElement))
   > Remove Index!(execute, TargetArray=Get CurrentPaths(), IndexToRemove:int=@n4) -> (then)
   Set l_path(execute=@n3.LoopBody, l_path=Replace(SourceString=@n3, From="\\", To=/))
   > Set l_index(l_index:int=@n3.Array Index) -> (Output_Get:int=0)
   > n5 = Switch on TA_FilePicker_Scope_Enum(Selection=Get pathScope() -> (pathScope=NewEnumerator0))
   n6 = IsRelative(InPath=Get l_path())
   > Branch(Condition=@n6)
   > Set Array Elem!(
    execute
    TargetArray=Get CurrentPaths()
    Index:int=Get l_index() -> (l_index:int=0)
    Item:"String (by ref)"=ConvertToAbsolutePath(Filename=Get l_path())
    bSizeToFit:bool=false
   ) -> (then)
   n7 = Branch(execute=@n5.NewEnumerator1, Condition=@n6)
   Set Array Elem(
    execute=@n7.else
    TargetArray=Get CurrentPaths()
    Index=Get l_index() -> (l_index=0)
    Item=ConvertToRelativePath!(Filename:String=Get l_path()) -> (ReturnValue:String)
    bSizeToFit=false
   )
   n8 = convertFilePathToAsset!(execute=@n5.NewEnumerator2, self, path:String=Get l_path()) -> (then, asset:String) {is=title K2Node_CallFunction}
   > Set Array Elem(TargetArray=Get CurrentPaths(), Index=Get l_index() -> (l_index=0), Item=@n8, bSizeToFit=false)
   validatePathText(execute=@n2.then_1)
   > validateToolTip()
   n9 = Get CurrentPaths()
   > n10 = Branch(Condition=Is Valid Index(TargetArray=@n9))
   n11 = Self-Reference()
   > Call On Paths Changed!(
    execute:Exec
    self:File Picker Base EUW Object Reference
    widget:File Picker Base EUW Object Reference=@n11
    paths:Array of Strings=@n9
    firstPath:String="Get (a copy)"(Array=@n9)
   ) -> (then:Exec) {is=title K2Node_CallDelegate}
   Call On Paths Changed(execute=@n10.else, widget=@n11, paths=Make Array())
  ]}
 setLabel {nodes [
   n1 = setLabel!() -> (then, label:String) {is=title K2Node_FunctionEntry}
   > Set currentLabel(currentLabel:String=@n1) -> (Output_Get:String)
   > Is Valid(InputObject=getLabelWidget!() -> (widget:Widget Object Reference) {is=title K2Node_CallFunction})
   > n2 = Cast To TextBlock!(execute, Object=getLabelWidget()) -> (then, CastFailed, AsText:Text Object Reference) {is=title K2Node_DynamicCast}
   > "SetText (Text)"(self=@n2, InText="To Text (String)"(InString=Get currentLabel() -> (currentLabel:String)))
  ]}
 setPickerEnabled {nodes [
   n1 = setPickerEnabled!() -> (then, Enabled:bool=~) {is=title K2Node_FunctionEntry}
   > Branch(Condition="Not Equal (Boolean)"(A=@n1, B=Get FilePickingEnabled() -> (FilePickingEnabled:bool)) -> (ReturnValue=~))
   > Set FilePickingEnabled(FilePickingEnabled:bool=@n1) -> (Output_Get:bool)
   > validateEnabledVisuals()
  ]}
 validateEnabledVisuals {nodes [
   validateEnabledVisuals() {is=title K2Node_FunctionEntry}
   > n1 = Sequence!(execute) -> (then_0, then_1, then_2:Exec)
   n2 = getButtonWidget() {is=title K2Node_CallFunction}
   > Is Valid(InputObject=@n2)
   > SetIsEnabled(self=@n2, bInIsEnabled=Get FilePickingEnabled())
   n3 = getInputWidget() {is=title K2Node_CallFunction}
   Is Valid(exec=@n1.then_1, InputObject=@n3)
   > n4 = Cast To EditableText(Object=@n3)
   n5 = Get FilePickingEnabled()
   > "SetIsReadOnly (Editable Text)"!(execute, self:Editable Text Object Reference=@n4, InbIsReadyOnly:bool=NOT Boolean(A=@n5)) -> (then)
   SetIsEnabled(execute=@n4.CastFailed, self=@n3, bInIsEnabled=@n5)
   n6 = getAssetPickerButton() {is=title K2Node_CallFunction}
   Is Valid(exec=@n1.then_2, InputObject=@n6)
   n7 = Get pickerType() -> (pickerType=NewEnumerator0)
   > SetVisibility(
    self=@n6
    InVisibility=Select(
     Option 0=Collapsed
     Option 1=Select!(NewEnumerator0:ESlateVisibility Enum=Visible, NewEnumerator1:ESlateVisibility Enum=Collapsed, Index:TA_FilePicker_Type_Enum Enum=@n7) -> (ReturnValue)
     Index=Get ShouldShowPickerButton() -> (ShouldShowPickerButton:bool)
    )
   )
  ]}
 validatePathText {nodes [
   validatePathText() {is=title K2Node_FunctionEntry}
   > n1 = Cast To EditableText(Object=getInputWidget())
   > Set l_editableText(l_editableText:Editable Text Object Reference=@n1) -> (Output_Get:Editable Text Object Reference)
   > n2 = getPathStringToDisplay!(execute, self) -> (then, string:String) {is=title K2Node_CallFunction}
   > "SetText (Editable Text)"!(
    execute
    self:Editable Text Object Reference=Get l_editableText() -> (l_editableText:Editable Text Object Reference)
    InText:Text="To Text (String)"(InString=@n2)
   ) -> (then)
  ]}
 validateToolTip {nodes [
   validateToolTip() {is=title K2Node_FunctionEntry}
   > Is Valid(InputObject=getInputWidget())
   > n1 = Sequence!(execute) -> (then_0, then_1)
   > n2 = For Each Loop(Array=Get CurrentPaths())
   n3 = "Equal (Integer)"(A=@n2.Array Index)
   > Set l_tooltip(l_tooltip:String=Append(A=Get l_tooltip() -> (l_tooltip:String), B=SelectString(B="\r\n", bPickA=@n3), C=@n2)) -> (Output_Get:String)
   SetToolTipText!(
    execute=@n1.then_1
    self:Widget Object Reference=getInputWidget()
    InToolTipText:"Text (by ref)"="To Text (String)"(InString=Get l_tooltip())
   ) -> (then)
  ]}
]
variables [
 allowMultipleFilesSelected {type=boolean}
 currentLabel {type=string, default=File}
 CurrentPaths {type="array<string>", default []}
 FilePickingEnabled {type=boolean, default=true}
 fileTypes {type="map<string>", default {}}
 initialDirectoryOverride {type=string, default=""}
 initialDirectoryType {type=TA_FilePicker_InitDirectoryType_Enum, default=Current File Picker Path}
 onPathsChanged {default=()}
 pathScope {type=TA_FilePicker_Scope_Enum, default=Absolute}
 pickerType {type=TA_FilePicker_Type_Enum, default=File}
 PopupTitle {type=string, default=""}
 ShouldShowPickerButton {type=boolean}
 startingPaths {type="array<string>", default []}
 WidgetLabelOverride {type=string, default=File}
]
widgets {info {rootWidgetClass=null, widgetCount=0}, widgets []}
