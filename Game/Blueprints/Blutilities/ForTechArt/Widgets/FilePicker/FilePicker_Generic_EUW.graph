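# Editor utility widget for tech art tools: a generic one-row file picker with a label, editable path field, asset-choose arrow button and browse button.
# It derives from FilePicker_Base_EUW and holds no picking logic itself, only calling the parent Construct.
# Its functions set the label and path text and hand the label, input field and both buttons back through getter functions, apparently for the base class to use.
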EditorUtilityWidgetBlueprint /Game/Blueprints/Blutilities/ForTechArt/Widgets/FilePicker/FilePicker_Generic_EUW
graphs [
 EventGraph {nodes [
   Event Construct()
   > Parent: Construct()
  ]}
 SetFilePath {nodes [
   n1 = SetFilePath!() -> (then, Path:String) {is=title K2Node_FunctionEntry}
   > "SetText (Editable Text)"!(
    execute
    self:Editable Text Object Reference=Get editTextPath() -> (editTextPath:Editable Text Object Reference)
    InText:Text="To Text (String)"(InString=@n1)
   ) -> (then)
  ]}
 SetTitleText {nodes [
   n1 = SetTitleText!() -> (then, Title:String) {is=title K2Node_FunctionEntry}
   > Set Text(Text="To Text (String)"(InString=@n1), self=Get textLabel() -> (textLabel:Common Text Object Reference))
  ]}
 getAssetPickerButton {nodes [
   getAssetPickerButton() {is=title K2Node_FunctionEntry}
   > Return Node!(execute, widget:Widget Object Reference=Get btnChooseAsset() -> (btnChooseAsset:TA Generic 01 Button Object Reference))
  ]}
 getButtonWidget {nodes [
   getButtonWidget() {is=title K2Node_FunctionEntry}
   > Return Node!(execute, button:Widget Object Reference=Get btnFilePick() -> (btnFilePick:TA Generic 01 Button Object Reference))
  ]}
 getInputWidget {nodes [
   getInputWidget() {is=title K2Node_FunctionEntry}
   > Return Node!(execute, widget:Widget Object Reference=Get editTextPath())
  ]}
 getLabelWidget {nodes [
   getLabelWidget() {is=title K2Node_FunctionEntry}
   > Return Node(widget=Get textLabel())
  ]}
]
variables []
widgets {
 info {
  parentClass=&/FilePicker_Base_EUW.FilePicker_Base_EUW_C
  rootWidgetClass=&Border
  widgetCount=10
 }
 widgets [
  Border_0 {widgetClassPath=&Border, widget=Border_0, properties {brushColor=<rgba 0.0212 0.0212 0.0212 1>}}
  VerticalBox_0 {widgetClassPath=&VerticalBox, parent=Border_0, widget=VerticalBox_0, slot {class=BorderSlot}}
  HorizontalBox_0 {
   widgetClassPath=&HorizontalBox
   parent=VerticalBox_0
   widget=HorizontalBox_0
   slot {class=VerticalBoxSlot, properties {padding=<ltrb 0 2 0 2>}}
  }
  textLabel {
   widgetClassPath=&CommonTextBlock
   bIsVariable=true
   parent=HorizontalBox_0
   widget=textLabel
   properties {
    text=File
    colorAndOpacity {colorUseRule=UseColor_Specified, specifiedColor=<rgba 0.4564 0.4564 0.4564 1>}
    font {
     bForceMonospaced=false
     bMaterialIsStencil=false
     fontMaterial=null
     fontObject=&NotoSans
     letterSpacing=0
     monospacedWidth=1
     size=10
     skewAmount=0
     typefaceFontName=NotoSans-Regular
     outlineSettings {
      bApplyOutlineToDropShadows=false
      bMiteredCorners=false
      bSeparateFillAlpha=false
      outlineMaterial=null
      outlineSize=0
      outlineColor=<rgba 0 0 0 1>
     }
    }
    shadowOffset=<0 0>
   }
   slot {class=HorizontalBoxSlot, properties {verticalAlignment=VAlign_Center}}
  }
  divider0 {
   widgetClassPath=&Image
   parent=HorizontalBox_0
   widget=divider0
   properties {
    brush {
     bIsDynamicallyLoaded=false
     drawAs=Image
     imageType=NoImage
     mirroring=NoMirror
     resourceName=None
     resourceObject=null
     tiling=NoTile
     imageSize=<1 1>
     margin=<ltrb 0 0 0 0>
     outlineSettings {
      bUseBrushTransparency=false
      roundingType=HalfHeightRadius
      width=0
      color {colorUseRule=UseColor_Specified, specifiedColor=<rgba 0 0 0 0>}
      cornerRadii=<0 0 0 0>
     }
     tintColor {colorUseRule=UseColor_Specified, specifiedColor=<rgba 1 1 1 1>}
     uVRegion {bIsValid=false, max=<0 0>, min=<0 0>}
    }
    colorAndOpacity=<rgba 0.0075 0.0075 0.0075 1>
   }
   slot {class=HorizontalBoxSlot, properties {padding=<ltrb 5 2 5 2>}}
  }
  Border_1 {
   widgetClassPath=&Border
   parent=HorizontalBox_0
   widget=Border_1
   properties {verticalAlignment=VAlign_Center, brushColor=<rgba 0.0048 0.0048 0.0048 1>}
   slot {class=HorizontalBoxSlot, properties {verticalAlignment=VAlign_Center, size {sizeRule=Fill, value=1}}}
  }
  editTextPath {
   widgetClassPath=/Script/UMG.EditableText
   bIsVariable=true
   parent=Border_1
   widget=editTextPath
   properties {
    minimumDesiredWidth=200
    text=path here..
    widgetStyle {
     backgroundImageComposing {
      bIsDynamicallyLoaded=false
      drawAs=Border
      imageType=FullColor
      mirroring=NoMirror
      resourceName=../../../Engine/Content/Slate/Old/HyperlinkDotted.png
      resourceObject=null
      tiling=Both
      imageSize=<0 0>
      margin=<ltrb 0 0 0 0.1875>
      outlineSettings {
       bUseBrushTransparency=false
       roundingType=HalfHeightRadius
       width=0
       color {colorUseRule=UseColor_Specified, specifiedColor=<rgba 0 0 0 0>}
       cornerRadii=<0 0 0 0>
      }
      tintColor {colorUseRule=UseColor_Specified, specifiedColor=<rgba 1 1 1 1>}
      uVRegion {bIsValid=false, max=<0 0>, min=<0 0>}
     }
     backgroundImageSelected {
      bIsDynamicallyLoaded=false
      drawAs=Image
      imageType=NoImage
      mirroring=NoMirror
      resourceName=None
      resourceObject=null
      tiling=NoTile
      imageSize=<0 0>
      margin=<ltrb 0 0 0 0>
      outlineSettings {
       bUseBrushTransparency=false
       roundingType=HalfHeightRadius
       width=0
       color {colorUseRule=UseColor_Specified, specifiedColor=<rgba 0 0 0 0>}
       cornerRadii=<0 0 0 0>
      }
      tintColor {colorUseRule=UseColor_Specified, specifiedColor=<rgba 0 0.162 0.7454 1>}
      uVRegion {bIsValid=false, max=<0 0>, min=<0 0>}
     }
     caretImage {
      bIsDynamicallyLoaded=false
      drawAs=Image
      imageType=NoImage
      mirroring=NoMirror
      resourceName=None
      resourceObject=null
      tiling=NoTile
      imageSize=<0 0>
      margin=<ltrb 0 0 0 0>
      outlineSettings {
       bUseBrushTransparency=false
       roundingType=HalfHeightRadius
       width=0
       color {colorUseRule=UseColor_Specified, specifiedColor=<rgba 0 0 0 0>}
       cornerRadii=<0 0 0 0>
      }
      tintColor {colorUseRule=UseColor_Specified, specifiedColor=<rgba 1 1 1 1>}
      uVRegion {bIsValid=false, max=<0 0>, min=<0 0>}
     }
     colorAndOpacity {colorUseRule=UseColor_Foreground, specifiedColor=<rgba 1 0 1 1>}
     font {
      bForceMonospaced=false
      bMaterialIsStencil=false
      fontMaterial=null
      fontObject=&NotoSans
      letterSpacing=0
      monospacedWidth=1
      size=10
      skewAmount=0
      typefaceFontName=NotoSans-Regular
      outlineSettings {
       bApplyOutlineToDropShadows=false
       bMiteredCorners=false
       bSeparateFillAlpha=false
       outlineMaterial=null
       outlineSize=0
       outlineColor=<rgba 0 0 0 1>
      }
     }
     hintColorAndOpacity {colorUseRule=UseColor_Foreground, specifiedColor=<rgba 1 0 1 1>}
     selectedBackgroundColor {colorUseRule=UseColor_Foreground, specifiedColor=<rgba 1 0 1 1>}
    }
   }
   slot {class=BorderSlot, properties {verticalAlignment=VAlign_Center}}
  }
  btnChooseAsset {
   widgetClassPath=&TA_Generic01_Button
   parent=HorizontalBox_0
   widget=btnChooseAsset
   properties {label="←", buttonColor_Default=<rgba 0.0497 0.0545 0.0685 1>}
   slot {class=HorizontalBoxSlot, properties {verticalAlignment=VAlign_Center, padding=<ltrb 5 0 5 0>}}
  }
  btnFilePick {
   widgetClassPath=&TA_Generic01_Button
   parent=HorizontalBox_0
   widget=btnFilePick
   properties {label=" ... ", buttonColor_Default=<rgba 0.0497 0.0545 0.0685 1>}
   slot {class=HorizontalBoxSlot, properties {verticalAlignment=VAlign_Center, padding=<ltrb 5 0 5 0>}}
  }
  divider0_1 {
   widgetClassPath=&Image
   parent=HorizontalBox_0
   widget=divider0_1
   properties {
    brush {
     bIsDynamicallyLoaded=false
     drawAs=Image
     imageType=NoImage
     mirroring=NoMirror
     resourceName=None
     resourceObject=null
     tiling=NoTile
     imageSize=<1 1>
     margin=<ltrb 0 0 0 0>
     outlineSettings {
      bUseBrushTransparency=false
      roundingType=HalfHeightRadius
      width=0
      color {colorUseRule=UseColor_Specified, specifiedColor=<rgba 0 0 0 0>}
      cornerRadii=<0 0 0 0>
     }
     tintColor {colorUseRule=UseColor_Specified, specifiedColor=<rgba 1 1 1 1>}
     uVRegion {bIsValid=false, max=<0 0>, min=<0 0>}
    }
    colorAndOpacity=<rgba 0.0075 0.0075 0.0075 1>
   }
   slot {class=HorizontalBoxSlot, properties {padding=<ltrb 5 2 5 2>}}
  }
 ]
}
